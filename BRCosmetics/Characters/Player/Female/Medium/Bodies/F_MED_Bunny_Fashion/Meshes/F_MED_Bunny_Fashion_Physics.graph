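# Physics asset for the female medium Bunny Fashion cosmetic body, defining secondary motion for its dangling strings and zipper.
# Capsules on head, pelvis and spine form the body volume.
# A four-segment left string, two-segment right string and boxed zipper hang from MSTR anchor bodies with tight twist limits.
# Left string and zipper anchors constrain to upper spine with free rotation, while the right anchor has no spine constraint.
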PhysicsAsset /BRCosmetics/Characters/Player/Female/Medium/Bodies/F_MED_Bunny_Fashion/Meshes/F_MED_Bunny_Fashion_Physics
bodies {
 dyn_l_string_01 [{shapeType=Capsule, length=3.9491, radius=0.8236, center=<2 0 0>, rotation=<rot 90 0 0>}]
 dyn_l_string_02 [{shapeType=Capsule, length=3.537, radius=0.7705, center=<3 0 0>, rotation=<rot 90 0 0>}]
 dyn_l_string_03 [{shapeType=Capsule, length=4.1524, radius=0.6648, center=<3 0 0>, rotation=<rot 90 0 0>}]
 dyn_l_string_04 [{shapeType=Capsule, length=2.9156, radius=1.2157, center=<3 0 0>, rotation=<rot 90 0 0>}]
 dyn_l_string_MSTR [{shapeType=Capsule, length=0.505, radius=0.505, rotation=<rot 0 -90 0>}]
 dyn_r_string_01 [{shapeType=Capsule, length=2.1512, radius=0.8941, center=<2 0 0>, rotation=<rot 90 0 0>}]
 dyn_r_string_02 [{shapeType=Capsule, length=1.6824, radius=1.2417, center=<2.5 0 0>, rotation=<rot 90 0 0>}]
 dyn_r_string_MSTR [{shapeType=Capsule, length=0.505, radius=0.505, rotation=<rot 0 -90 0>}]
 dyn_zipper [{shapeType=Box, extentX=1.2016, extentY=0.8325, extentZ=3.4706, center=<1.5478 -0.0143 0.0263>, rotation=<rot -9.9982 -89.6128 -88.8917>}]
 dyn_zipper_MSTR [
  {shapeType=Capsule, length=3.2848, radius=2.3172, center=<2.1787 -2.0862 3.0935>, rotation=<rot -49.8 26.9292 -7.9902>}
  {shapeType=Box, extentX=1.01, extentY=6.7147, extentZ=6.529, center=<0.2537 -0.4828 -1.3244>, rotation=<rot -15.8064 -67.0641 -4.3746>}
 ]
 head [{shapeType=Capsule, length=9.6638, radius=7.8147, center=<2.5341 0.2233 0>, rotation=<rot 0 -89.9996 89.9997>}]
 pelvis [{shapeType=Capsule, length=11.1331, radius=9.3964, center=<0 -2.266 0>}]
 spine_01 [{shapeType=Capsule, length=9.4519, radius=8.8085, rotation=<rot 0 -90 0>}]
 spine_02 [{shapeType=Capsule, length=4.063, radius=8.6027, center=<0 1.0648 0>}]
 spine_03 [{shapeType=Capsule, length=2.9995, radius=9.0682, center=<0.9083 2.1283 0>}]
 spine_04 [
  {shapeType=Capsule, length=7.7433, radius=8.4536, center=<3.5655 3.1859 0>, rotation=<rot 0 -16.5449 0>}
  {shapeType=Capsule, length=9.3578, radius=5.2086, center=<1.5105 7.1621 0>}
 ]
 spine_05 [{shapeType=Capsule, length=5.8678, radius=10.4654, center=<-1.0416 0.716 0>, rotation=<rot 0 -2.8965 0>}]
}
constraints [
 {bone1Name=spine_04, bone2Name=spine_03, swing1LimitDegrees=45, swing2LimitDegrees=45, twistLimitDegrees=45}
 {bone1Name=spine_05, bone2Name=spine_04, swing1LimitDegrees=45, swing2LimitDegrees=45, twistLimitDegrees=45}
 {
  bone1Name=spine_03
  bone2Name=spine_02
  swing1LimitDegrees=45
  swing1Motion=Free
  swing2LimitDegrees=45
  swing2Motion=Free
  twistLimitDegrees=45
  twistMotion=Free
 }
 {
  bone1Name=spine_02
  bone2Name=spine_01
  swing1LimitDegrees=45
  swing1Motion=Free
  swing2LimitDegrees=45
  swing2Motion=Free
  twistLimitDegrees=45
  twistMotion=Free
 }
 {
  bone1Name=spine_01
  bone2Name=pelvis
  swing1LimitDegrees=45
  swing1Motion=Free
  swing2LimitDegrees=45
  swing2Motion=Free
  twistLimitDegrees=45
  twistMotion=Free
 }
 {
  bone1Name=head
  bone2Name=spine_05
  swing1LimitDegrees=45
  swing1Motion=Free
  swing2LimitDegrees=45
  swing2Motion=Free
  twistLimitDegrees=45
  twistMotion=Free
 }
 {bone1Name=dyn_zipper, bone2Name=dyn_zipper_MSTR, swing1LimitDegrees=45, swing2LimitDegrees=45, twistLimitDegrees=5}
 {
  bone1Name=dyn_zipper_MSTR
  bone2Name=spine_04
  swing1LimitDegrees=45
  swing1Motion=Free
  swing2LimitDegrees=45
  swing2Motion=Free
  twistLimitDegrees=45
  twistMotion=Free
 }
 {
  bone1Name=dyn_l_string_MSTR
  bone2Name=spine_05
  swing1LimitDegrees=45
  swing1Motion=Free
  swing2LimitDegrees=45
  swing2Motion=Free
  twistLimitDegrees=45
  twistMotion=Free
 }
 {bone1Name=dyn_l_string_01, bone2Name=dyn_l_string_MSTR, swing1LimitDegrees=35, swing2LimitDegrees=35, twistLimitDegrees=5}
 {bone1Name=dyn_r_string_02, bone2Name=dyn_r_string_01, swing1LimitDegrees=35, swing2LimitDegrees=35, twistLimitDegrees=5}
 {bone1Name=dyn_r_string_01, bone2Name=dyn_r_string_MSTR, swing1LimitDegrees=35, swing2LimitDegrees=35, twistLimitDegrees=5}
 {bone1Name=dyn_l_string_02, bone2Name=dyn_l_string_01, swing1LimitDegrees=35, swing2LimitDegrees=35, twistLimitDegrees=5}
 {bone1Name=dyn_l_string_03, bone2Name=dyn_l_string_02, swing1LimitDegrees=35, swing2LimitDegrees=35, twistLimitDegrees=5}
 {bone1Name=dyn_l_string_04, bone2Name=dyn_l_string_03, swing1LimitDegrees=35, swing2LimitDegrees=35, twistLimitDegrees=5}
]
