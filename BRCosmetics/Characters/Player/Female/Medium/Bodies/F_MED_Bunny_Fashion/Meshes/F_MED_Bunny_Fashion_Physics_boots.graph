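# Physics asset for the boots of the F_MED_Bunny_Fashion player cosmetic body, setting up secondary motion for shoe ties and a pouch.
# Each shoe master bone has a tiny capsule with free angular constraints to its calf twist bone, which carries a larger capsule.
# Four front and back tie capsules per shoe hang from that master with limited swing and tight twist.
# A box-shaped pouch is constrained within a few degrees to its own master body.
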PhysicsAsset /BRCosmetics/Characters/Player/Female/Medium/Bodies/F_MED_Bunny_Fashion/Meshes/F_MED_Bunny_Fashion_Physics_boots
bodies {
 calf_twist_02_l [{shapeType=Capsule, length=11.8918, radius=5.433, center=<-3.8549 -1.7698 -1.062>, rotation=<rot 90 0 0>}]
 calf_twist_02_r [{shapeType=Capsule, length=11.7532, radius=5.5715, center=<3.8548 1.7698 1.1631>, rotation=<rot 90 0 0>}]
 dyn_left_shoe_MSTR [{shapeType=Capsule, length=0.2, radius=0.2, center=<3.3769 -0.4076 0.6617>, rotation=<rot 0 83.9086 -82.4065>}]
 dyn_lt_shoe_tie_bk_01 [{shapeType=Capsule, length=1.357, radius=1.1322, center=<3 0 0>}]
 dyn_lt_shoe_tie_bk_02 [{shapeType=Capsule, length=1.357, radius=1.1322, center=<3 0 0>}]
 dyn_lt_shoe_tie_fr_01 [{shapeType=Capsule, length=5.9024, radius=0.7541, center=<5 0 0>, rotation=<rot 90 0 0>}]
 dyn_lt_shoe_tie_fr_02 [{shapeType=Capsule, length=5.9024, radius=0.7541, center=<5 0 0>, rotation=<rot 90 0 0>}]
 dyn_pouch [{shapeType=Box, extentX=4.3581, extentY=10.6081, extentZ=16.7211, center=<1.6853 0 0>}]
 dyn_pouch_MSTR [{shapeType=Box, extentX=1.01, extentY=1.01, extentZ=1.01, rotation=<rot 0 -90 0>}]
 dyn_right_shoe_MSTR [{shapeType=Capsule, length=0.2, radius=0.2, center=<3.3769 0.4076 0.6617>, rotation=<rot 0 96.0914 -82.4066>}]
 dyn_rt_shoe_tie_bk_01 [{shapeType=Capsule, length=1.357, radius=1.1322, center=<3 0 0>}]
 dyn_rt_shoe_tie_bk_02 [{shapeType=Capsule, length=1.357, radius=1.1322, center=<3 0 0>}]
 dyn_rt_shoe_tie_fr_01 [{shapeType=Capsule, length=5.9024, radius=0.7541, center=<5 0 0>, rotation=<rot 90 0 0>}]
 dyn_rt_shoe_tie_fr_02 [{shapeType=Capsule, length=5.9024, radius=0.7541, center=<5 0 0>, rotation=<rot 90 0 0>}]
}
constraints [
 {
  bone1Name=dyn_left_shoe_MSTR
  bone2Name=calf_twist_02_l
  swing1LimitDegrees=45
  swing1Motion=Free
  swing2LimitDegrees=45
  swing2Motion=Free
  twistLimitDegrees=45
  twistMotion=Free
 }
 {bone1Name=dyn_lt_shoe_tie_bk_01, bone2Name=dyn_left_shoe_MSTR, swing1LimitDegrees=20, swing2LimitDegrees=35, twistLimitDegrees=5}
 {bone1Name=dyn_lt_shoe_tie_fr_01, bone2Name=dyn_left_shoe_MSTR, swing1LimitDegrees=35, swing2LimitDegrees=35, twistLimitDegrees=5}
 {bone1Name=dyn_lt_shoe_tie_fr_02, bone2Name=dyn_left_shoe_MSTR, swing1LimitDegrees=35, swing2LimitDegrees=35, twistLimitDegrees=5}
 {bone1Name=dyn_lt_shoe_tie_bk_02, bone2Name=dyn_left_shoe_MSTR, swing1LimitDegrees=20, swing2LimitDegrees=35, twistLimitDegrees=5}
 {
  bone1Name=dyn_right_shoe_MSTR
  bone2Name=calf_twist_02_r
  swing1LimitDegrees=45
  swing1Motion=Free
  swing2LimitDegrees=45
  swing2Motion=Free
  twistLimitDegrees=45
  twistMotion=Free
 }
 {bone1Name=dyn_rt_shoe_tie_bk_01, bone2Name=dyn_right_shoe_MSTR, swing1LimitDegrees=20, swing2LimitDegrees=35, twistLimitDegrees=5}
 {bone1Name=dyn_rt_shoe_tie_fr_01, bone2Name=dyn_right_shoe_MSTR, swing1LimitDegrees=35, swing2LimitDegrees=35, twistLimitDegrees=5}
 {bone1Name=dyn_rt_shoe_tie_fr_02, bone2Name=dyn_right_shoe_MSTR, swing1LimitDegrees=35, swing2LimitDegrees=35, twistLimitDegrees=5}
 {bone1Name=dyn_rt_shoe_tie_bk_02, bone2Name=dyn_right_shoe_MSTR, swing1LimitDegrees=20, swing2LimitDegrees=35, twistLimitDegrees=5}
 {bone1Name=dyn_pouch, bone2Name=dyn_pouch_MSTR, swing1LimitDegrees=5, swing2LimitDegrees=2, twistLimitDegrees=3}
]
