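# Animation Blueprint for the female medium Bunny Fashion cosmetic body, copying the attached parent mesh's pose and adding secondary deformation and motion.
# The copied pose passes through two Control Rigs, then additive glute corrective bones driven by thigh-versus-pelvis twist alphas computed each update.
# Weighted multi-bone constraints anchor string, zipper, shoe and pouch master bones to the body.
# Two base-bone-space RigidBody passes then simulate dynamics from separate physics assets, relative to spine_02 and root.
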AnimBlueprint /BRCosmetics/Characters/Player/Female/Medium/Bodies/F_MED_Bunny_Fashion/Meshes/F_MED_Bunny_Fashion_AnimBP
target_skeleton=&/F_MED_Bunny_Fashion_Skeleton
graphs [
 AnimGraph {nodes [
   n1 = Control Rig(Source=Copy Pose From Mesh() {bUseAttachedParent=true}) {
    alphaBoolBlend {blendOption=Linear}
    controlRigAssetReference {blueprintRigClass=/BRCosmetics/Characters/Player/Female/Medium/Bodies/F_MED_Bunny_Fashion/Meshes/F_MED_Bunny_Fashion_CtrlRig/F_MED_Bunny_Fashion_CtrlRig.F_MED_Bunny_Fashion_CtrlRig_C}
   }
   n2 = Control Rig(Source=@n1) {
    alphaBoolBlend {blendOption=Linear}
    controlRigAssetReference {blueprintRigClass=&/F_MED_Bunny_Fashion_DeformRig_Master_ControlRig.F_MED_Bunny_Fashion_DeformRig_Master_ControlRig_C}
   }
   n3 = "Transform (Modify) Bone - Bone: dyn_glute_l"(
    ComponentPose=Local To Component(LocalPose=@n2)
    Alpha=Get thigh_rot_l()
    Translation="-4.689,5.014,-4.028"
    Rotation="0.000,0.000,-55"
    Scale="1.25,1.25,1.25"
   ) {
    rotationMode=BMM_Additive
    rotationSpace=BCS_BoneSpace
    scaleMode=BMM_Additive
    scaleSpace=BCS_BoneSpace
    translationMode=BMM_Additive
    translationSpace=BCS_BoneSpace
    alphaBoolBlend {blendOption=Linear}
    boneToModify {boneName=dyn_glute_l}
    rotation {roll=-55}
    scale=<1.25 1.25 1.25>
    translation=<-4.689 5.014 -4.028>
   }
   n4 = "Transform (Modify) Bone - Bone: dyn_glute_r"(ComponentPose=@n3, Alpha=Get thigh_rot_r(), Translation="-3.75,0,0", Rotation="0.000,0.000,-55", Scale="1.3,1.3,1.3") {
    rotationMode=BMM_Additive
    rotationSpace=BCS_BoneSpace
    scaleMode=BMM_Additive
    scaleSpace=BCS_BoneSpace
    translationMode=BMM_Additive
    translationSpace=BCS_BoneSpace
    alphaBoolBlend {blendOption=Linear}
    boneToModify {boneName=dyn_glute_r}
    rotation {roll=-55}
    scale=<1.3 1.3 1.3>
    translation {x=-3.75}
   }
   n5 = "Constraint - dyn_r_string_MSTR (PPPPPPPP)"!(
    ComponentPose=@n4
    Alpha=1
    ConstraintWeights_0=0.0281
    ConstraintWeights_1:float=0.5771
    ConstraintWeights_2:float=0.0855
    ConstraintWeights_3:float=0.1208
    ConstraintWeights_4:float=0.0922
    ConstraintWeights_5:float=0.0306
    ConstraintWeights_6:float=0.0628
    ConstraintWeights_7:float=0.0029
   ) -> (Pose) {
    alphaBoolBlend {blendOption=Linear}
    boneToModify {boneName=dyn_r_string_MSTR}
    constraintSetup [
     {targetBone {boneName=spine_04}}
     {targetBone {boneName=spine_05}}
     {targetBone {boneName=clavicle_r}}
     {targetBone {boneName=neck_02}}
     {targetBone {boneName=neck_01}}
     {targetBone {boneName=dyn_collar_front}}
     {targetBone {boneName=dyn_collar_front_01}}
     {targetBone {boneName=dyn_collar_back}}
    ]
    constraintWeights [1,1,1,1,1,1,1,1]
   }
   n6 = "Constraint - dyn_l_string_MSTR (PPPPPPPP)"!(
    ComponentPose=@n5
    Alpha=1
    ConstraintWeights_0=0.0634
    ConstraintWeights_1:float=0.0459
    ConstraintWeights_2:float=0.604
    ConstraintWeights_3:float=0.1245
    ConstraintWeights_4:float=0.0943
    ConstraintWeights_5:float=0.0217
    ConstraintWeights_6:float=0.0443
    ConstraintWeights_7:float=0.0019
   ) -> (Pose) {
    alphaBoolBlend {blendOption=Linear}
    boneToModify {boneName=dyn_l_string_MSTR}
    constraintSetup [
     {targetBone {boneName=clavicle_l}}
     {targetBone {boneName=spine_04}}
     {targetBone {boneName=spine_05}}
     {targetBone {boneName=neck_02}}
     {targetBone {boneName=neck_01}}
     {targetBone {boneName=dyn_collar_front}}
     {targetBone {boneName=dyn_collar_front_01}}
     {targetBone {boneName=clavicle_r}}
    ]
    constraintWeights [1,1,1,1,1,1,1,1]
   }
   n7 = "Constraint - dyn_zipper_MSTR (PPPPPP)"!(
    ComponentPose=@n6
    Alpha=1
    ConstraintWeights_0=0.0034
    ConstraintWeights_1:float=0.0007
    ConstraintWeights_2:float=0.0054
    ConstraintWeights_3:float=0.9023
    ConstraintWeights_4:float=0.08
    ConstraintWeights_5:float=0.0082
   ) -> (Pose) {
    alphaBoolBlend {blendOption=Linear}
    boneToModify {boneName=dyn_zipper_MSTR}
    constraintSetup [
     {targetBone {boneName=upperarm_twist_01_r}}
     {targetBone {boneName=spine_02}}
     {targetBone {boneName=spine_03}}
     {targetBone {boneName=spine_04}}
     {targetBone {boneName=spine_05}}
     {targetBone {boneName=clavicle_r}}
    ]
    constraintWeights [1,1,1,1,1,1]
   }
   n8 = "Constraint - dyn_right_shoe_MSTR (PP)"!(ComponentPose=@n7, Alpha=1, ConstraintWeights_0=0.7128, ConstraintWeights_1:float=0.2872) -> (Pose) {
    alphaBoolBlend {blendOption=Linear}
    boneToModify {boneName=dyn_right_shoe_MSTR}
    constraintSetup [{targetBone {boneName=calf_twist_02_r}}, {targetBone {boneName=calf_r}}]
    constraintWeights [1,1]
   }
   n9 = "Constraint - dyn_left_shoe_MSTR (PP)"!(ComponentPose=@n8, Alpha=1, ConstraintWeights_0=0.2542, ConstraintWeights_1:float=0.7458) -> (Pose) {
    alphaBoolBlend {blendOption=Linear}
    boneToModify {boneName=dyn_left_shoe_MSTR}
    constraintSetup [{targetBone {boneName=calf_l}}, {targetBone {boneName=calf_twist_02_l}}]
    constraintWeights [1,1]
   }
   n10 = "Constraint - dyn_pouch_MSTR (PPP)"!(ComponentPose=@n9, Alpha=1, ConstraintWeights_0=0.5936, ConstraintWeights_1:float=0.3416, ConstraintWeights_2:float=0.0648) -> (Pose) {
    alphaBoolBlend {blendOption=Linear}
    boneToModify {boneName=dyn_pouch_MSTR}
    constraintSetup [{targetBone {boneName=thigh_r}}, {targetBone {boneName=thigh_twist_01_r}}, {targetBone {boneName=pelvis}}]
    constraintWeights [1,1,1]
   }
   n11 = RigidBody(ComponentPose=@n10) {
    bOverrideWorldGravity=true
    overridePhysicsAsset=&/F_MED_Bunny_Fashion_Physics
    simulationSpace=BaseBoneSpace
    alphaBoolBlend {blendOption=Linear}
    baseBoneRef {boneName=spine_02}
    componentLinearAccScale=<0.3 0.3 0.5>
    componentLinearVelScale=<0.2 0.2 0.3>
    overrideWorldGravity {z=-980}
   }
   n12 = RigidBody(ComponentPose=@n11) {
    bOverrideWorldGravity=true
    overridePhysicsAsset=&/F_MED_Bunny_Fashion_Physics_boots
    simulationSpace=BaseBoneSpace
    alphaBoolBlend {blendOption=Linear}
    baseBoneRef {boneName=root}
    componentLinearAccScale=<0.3 0.3 0.5>
    componentLinearVelScale=<0.2 0.2 0.3>
    overrideWorldGravity {z=-980}
   }
   Output Pose(Result=Component To Local(ComponentPose=@n12)) {layerGroup=None}
  ]}
 EventGraph {nodes [
   Event BlueprintUpdateAnimation()
   > Get!(execute:Exec) -> (then:Exec, else:Exec, Skel_Mesh:Skeletal Mesh Component Object Reference)
   > Sequence()
   n1 = Get Skel_Mesh()
   > Set thigh_rot_l(thigh_rot_l=GetAlphaFromReferenceTransformDelta(SkeletalMeshComponent=@n1, BoneName=thigh_l, BaseName=pelvis, InputRange=(X=-20,Y=-100), TwistAxis="1,0,0"))
   > Set thigh_rot_r(thigh_rot_r=GetAlphaFromReferenceTransformDelta(SkeletalMeshComponent=@n1, BoneName=thigh_r, BaseName=pelvis, InputRange=(X=-20,Y=-100), TwistAxis="1,0,0"))
   Event BlueprintInitializeAnimation()
   > Set Skel_Mesh(Skel_Mesh=GetOwnerSkeletalMeshComponent())
  ]}
]
variables [
 Skel_Mesh {type=SkeletalMeshComponent}
 thigh_rot_l {type=number}
 thigh_rot_r {type=number}
]
