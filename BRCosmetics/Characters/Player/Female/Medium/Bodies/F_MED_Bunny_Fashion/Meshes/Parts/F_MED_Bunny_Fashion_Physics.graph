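# Physics asset for the female medium Bunny Fashion cosmetic character, covering dynamic hood and ear bones around the head.
# Every body is a single capsule, with three hood bodies and both ear master bodies constrained directly to the head.
# Each ear master swings and twists freely, feeding a chain of tightly limited segments.
# The ears are asymmetric: four segments on the left, five on the right with mostly tighter limits.
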PhysicsAsset /BRCosmetics/Characters/Player/Female/Medium/Bodies/F_MED_Bunny_Fashion/Meshes/Parts/F_MED_Bunny_Fashion_Physics
bodies {
 dyn_hood_l [{shapeType=Capsule, length=1.7835, radius=1.4537, center=<4.8638 0 0>, rotation=<rot 0 0 90>}]
 dyn_hood_r [{shapeType=Capsule, length=1.7835, radius=1.4537, center=<4.8638 0 0>, rotation=<rot 0 0 90>}]
 dyn_hood_top [{shapeType=Capsule, length=2.6575, radius=2.2818, center=<3 0 0>}]
 dyn_l_ear_01 [{shapeType=Capsule, length=4.7401, radius=2.1399, center=<3 0 0>, rotation=<rot 90 0 0>}]
 dyn_l_ear_02 [{shapeType=Capsule, length=7.6651, radius=3.0803, center=<4.6066 0 0>, rotation=<rot 90 0 0>}]
 dyn_l_ear_03 [{shapeType=Capsule, length=6.6431, radius=3.379, center=<3 0 0>, rotation=<rot 90 0 0>}]
 dyn_l_ear_04 [{shapeType=Capsule, length=5.4337, radius=4.0817, center=<4 0 0>, rotation=<rot 90 0 0>}]
 dyn_l_ear_MSTR [{shapeType=Capsule, length=0.505, radius=0.505, rotation=<rot 0 -90 0>}]
 dyn_r_ear_01 [{shapeType=Capsule, length=5.8145, radius=1.599, center=<4 0 0>, rotation=<rot 90 0 0>}]
 dyn_r_ear_02 [{shapeType=Capsule, length=5.259, radius=2.3911, center=<3 0 0>, rotation=<rot 90 0 0>}]
 dyn_r_ear_03 [{shapeType=Capsule, length=7.3607, radius=4.852, center=<5 0 0>, rotation=<rot 90 0 0>}]
 dyn_r_ear_04 [{shapeType=Capsule, length=6.5076, radius=5.0888, center=<6.0307 0 0>, rotation=<rot 90 0 0>}]
 dyn_r_ear_05 [{shapeType=Capsule, length=2.772, radius=2.3907, center=<2 0 0>, rotation=<rot 90 0 0>}]
 dyn_r_ear_MSTR [{shapeType=Capsule, length=0.505, radius=0.505, rotation=<rot 0 -90 0>}]
 head [{shapeType=Capsule, length=6.7633, radius=8.4365, center=<5.5796 0.3851 0.2103>, rotation=<rot -8.6946 -90 -92.5096>}]
}
constraints [
 {bone1Name=dyn_hood_r, bone2Name=head, swing1LimitDegrees=3, swing2LimitDegrees=5, twistLimitDegrees=5}
 {bone1Name=dyn_hood_l, bone2Name=head, swing1LimitDegrees=3, swing2LimitDegrees=5, twistLimitDegrees=5}
 {bone1Name=dyn_hood_top, bone2Name=head, swing1LimitDegrees=20, swing2LimitDegrees=20, twistLimitDegrees=3}
 {
  bone1Name=dyn_r_ear_MSTR
  bone2Name=head
  swing1LimitDegrees=45
  swing1Motion=Free
  swing2LimitDegrees=45
  swing2Motion=Free
  twistLimitDegrees=45
  twistMotion=Free
 }
 {bone1Name=dyn_r_ear_01, bone2Name=dyn_r_ear_MSTR, swing1LimitDegrees=1, swing2LimitDegrees=1, twistLimitDegrees=1}
 {bone1Name=dyn_r_ear_02, bone2Name=dyn_r_ear_01, swing1LimitDegrees=3, swing2LimitDegrees=3, twistLimitDegrees=1}
 {bone1Name=dyn_r_ear_03, bone2Name=dyn_r_ear_02, swing1LimitDegrees=3, swing2LimitDegrees=15, twistLimitDegrees=3}
 {bone1Name=dyn_r_ear_04, bone2Name=dyn_r_ear_03, swing1LimitDegrees=3, swing2LimitDegrees=10, twistLimitDegrees=3}
 {bone1Name=dyn_r_ear_05, bone2Name=dyn_r_ear_04, swing1LimitDegrees=3, swing2LimitDegrees=5, twistLimitDegrees=3}
 {
  bone1Name=dyn_l_ear_MSTR
  bone2Name=head
  swing1LimitDegrees=45
  swing1Motion=Free
  swing2LimitDegrees=45
  swing2Motion=Free
  twistLimitDegrees=45
  twistMotion=Free
 }
 {bone1Name=dyn_l_ear_01, bone2Name=dyn_l_ear_MSTR, swing1LimitDegrees=5, swing2LimitDegrees=10, twistLimitDegrees=3}
 {bone1Name=dyn_l_ear_02, bone2Name=dyn_l_ear_01, swing1LimitDegrees=5, swing2LimitDegrees=12, twistLimitDegrees=3}
 {bone1Name=dyn_l_ear_03, bone2Name=dyn_l_ear_02, swing1LimitDegrees=5, swing2LimitDegrees=15, twistLimitDegrees=3}
 {bone1Name=dyn_l_ear_04, bone2Name=dyn_l_ear_03, swing1LimitDegrees=5, swing2LimitDegrees=17, twistLimitDegrees=3}
]
